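; View model for the Pro Vocals tooltip in the Pilgrim Quickplay Universal Part Picker, deciding whether the tooltip shows.
; It waits for the event router, then checks conditions immediately and on each PartPickerOpen event, tracking the listener handle.
; ShowTooltip requires ChickenParm enabled on the player controller component, the player in the backstage area, and non-compact UI.
; An invalid controller or component hides the tooltip.
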(event Sparks|ViewModel|EventBP_OnGlobalEventRouterAvailable
  (params (EventRouter :type Gameplay%20Event%20Router%20Component%20Object%20Reference) (PlayerState :type Fort%20Player%20State%20Object%20Reference) (OwningLocalPlayer :type Local%20Player%20Object%20Reference) (OwningPlayerController :type Fort%20Player%20Controller%20Object%20Reference))
  (bind _eventrouter (Pilgrim|PilgrimWaitForEventRouter OwningPlayerController)
    (:OnEventRouterAvailable
      (Utilities|FlowControl|Sequence
        (:then_0
          (bind _asynctaskproxy (GameplayEventRouter|StartListeningToEvent (Sparks|ViewModel|BPGetOwningPlayerController) _eventrouter "/Script/PilgrimCoreRuntime.PilgrimEvent_PartPickerOpen")
            (:then
              (Sparks|ViewModel|TrackEventHandle _asynctaskproxy))
            (:OnEventReceived
              (CallFunction|CheckTooltipConditions))))
        (:then_1
          (CallFunction|CheckTooltipConditions))))))

(fn CheckTooltipConditions
  (bind _returnvalue (Sparks|ViewModel|BPGetOwningPlayerController))
  (bind _returnvalue_2 (Actor|GetComponentbyClass _returnvalue "/PilgrimCore/Player/Components/BP_PilgrimPlayerControllerComponent.BP_PilgrimPlayerControllerComponent_C"))
  (Utilities|IsValid _returnvalue
    (:Is%20Valid
      (Utilities|IsValid _returnvalue_2
        (:Is%20Valid
          (Variables|Default|SetShowTooltip (Math|Boolean|ANDBoolean (Class|PilgrimPlayerControllerComponent|IsChickenParmEnabled _returnvalue_2) (Class|PilgrimQuickplayBPFL|IsatBackstageArea _returnvalue) (not (DisplayHelper|IsUsingCompactUI)))))
        (:Is%20Not%20Valid
          _)))
    (:Is%20Not%20Valid
      _))
  (Variables|Default|SetShowTooltip false))
