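; View model for the difficulty part picker in Pilgrim Quickplay, supplying the list of song difficulties and the index of the player's current one.
; Once the Pilgrim event router is available, it builds the list and rebuilds it on each player-state difficulty-changed event.
; Each enum value except None becomes an entry data object with localized text.
; The current selection comes from the PilgrimCorePlayerStateComponent, falling back to Easy when that is invalid.
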(event Sparks|ViewModel|EventBP_OnGlobalEventRouterAvailable
  (params (EventRouter :type Gameplay%20Event%20Router%20Component%20Object%20Reference) (PlayerState :type Fort%20Player%20State%20Object%20Reference) (OwningLocalPlayer :type Local%20Player%20Object%20Reference) (OwningPlayerController :type Fort%20Player%20Controller%20Object%20Reference))
  (bind _eventrouter (Pilgrim|PilgrimWaitForEventRouter OwningPlayerController)
    (:OnEventRouterAvailable
      (Variables|Default|SetPilgrim%20Event%20Router%20Ref _eventrouter)
      (Utilities|FlowControl|Sequence
        (:then_0
          (bind _asynctaskproxy (GameplayEventRouter|StartListeningToEvent (Sparks|ViewModel|BPGetOwningPlayerController) (Variables|Default|GetPilgrim%20Event%20Router%20Ref) "/Script/PilgrimCoreRuntime.PilgrimEvent_PlayerState_DifficultyChanged")
            (:then
              (Sparks|ViewModel|TrackEventHandle _asynctaskproxy))
            (:OnEventReceived
              (CallFunction|SettingListDifficulty))))
        (:then_1
          (CallFunction|SettingListDifficulty))))))

(fn CreateListDifficulties
  (bind _currentpart (CallFunction|GetCurrentDifficultySelection))
  (LocalVariables|SetCurrentDifficultySelected _currentpart)
  (bind _enumvalue (Utilities|Enum|ForEachEPilgrimSongDifficulty true)
    (:LoopBody
      (if (Utilities|Enum|NotEqual(Enum) _enumvalue "None")
        (bind _returnvalue (Game|ConstructObjectfromClass "/PilgrimQuickplay/Widgets/UniversalPartPicker/BP_PilgrimQuickplay_PartPicker_DifficultyEntryData.BP_PilgrimQuickplay_PartPicker_DifficultyEntryData_C" 0 true _enumvalue (CallFunction|DetermineTextDifficulty _enumvalue)))
        (bind _returnvalue_2 (Utilities|Array|Add (LocalVariables|GetListAvailableDifficulties) _returnvalue))
        ; We check if this entry is the current difficulty
        (if (Utilities|Enum|Equal(Enum) _enumvalue (LocalVariables|GetCurrentDifficultySelected))
          (Variables|Default|SetIndexOnListCurrentDifficulty _returnvalue_2))))
    (:then
      (return (LocalVariables|GetListAvailableDifficulties)))))

(fn DetermineTextDifficulty
  (params (SongMetaData :type EPilgrimSongDifficulty%20Enum :default "DifficultyEasy"))
  (return (Utilities|Select :Index SongMetaData :DifficultyEasy "NSLOCTEXT(\"[BCDC174ABB6CF8B6F3A0F7D396F348DD]\", \"D1CFC9B74380C9435C7822A113AC302A\", \"Easy\")" :DifficultyMedium "NSLOCTEXT(\"[BCDC174ABB6CF8B6F3A0F7D396F348DD]\", \"44EBDCCE4387AC34871A89990E66743C\", \"Medium\")" :DifficultyHard "NSLOCTEXT(\"[BCDC174ABB6CF8B6F3A0F7D396F348DD]\", \"28F7EC1A4556C3194DC8688F1D8A7FA3\", \"Hard\")" :DifficultyExpert "NSLOCTEXT(\"[BCDC174ABB6CF8B6F3A0F7D396F348DD]\", \"13B1AF0B46A16646D17EA8BC2FA93C3F\", \"Expert\")" :None "NSLOCTEXT(\"[BCDC174ABB6CF8B6F3A0F7D396F348DD]\", \"A3154B064818520BEA30FBA636D44FDF\", \"Easy\")")))

(fn GetCurrentDifficultySelection
  (bind _returnvalue (Sparks|ViewModel|BPGetOwningPlayerController))
  (bind _playerstate (Variables|/Script/Engine.Controller|GetPlayerState _returnvalue))
  (bind _returnvalue_2 (Actor|GetComponentbyClass _playerstate "/Script/PilgrimCoreRuntime.PilgrimCorePlayerStateComponent"))
  (Utilities|IsValid _returnvalue_2
    (:Is%20Valid
      (return (Class|PilgrimCorePlayerStateComponent|GetSongDifficulty _returnvalue_2)))
    (:Is%20Not%20Valid
      (return "DifficultyEasy"))))

(fn GetDifficultyText
  (return (CallFunction|DetermineTextDifficulty (Variables|Default|GetSongDifficluty))))

(fn SettingListDifficulty
  ; Setting the list of options (icons and names for the difficulties)
  (bind _difficultiesavailable (CallFunction|CreateListDifficulties))
  (Variables|Default|SetListOfDifficulties _difficultiesavailable))
