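; View model for the Pilgrim Quickplay universal part picker, supplying the selectable instrument entries with name, icon and song intensity.
; Event-router listeners for instrument, peripheral, setlist, picker and setting changes refresh the upcoming song's metadata and/or rebuild the list.
; Default and pro parts are filtered by pro eligibility, feature checks, cymbal setting and compact touch UI.
; BattleStage ScoreChase playlist tags instead limit the list to one instrument and its allowed pro variant.
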(event Sparks|ViewModel|EventBP_OnGlobalEventRouterAvailable
  (params (EventRouter :type Gameplay%20Event%20Router%20Component%20Object%20Reference) (PlayerState :type Fort%20Player%20State%20Object%20Reference) (OwningLocalPlayer :type Local%20Player%20Object%20Reference) (OwningPlayerController :type Fort%20Player%20Controller%20Object%20Reference))
  (CallFunction|SetListPartsInfo)
  (bind _eventrouter (Pilgrim|PilgrimWaitForEventRouter OwningPlayerController)
    (:OnEventRouterAvailable
      (Variables|Default|SetEvent%20Router%20Ref _eventrouter)
      (Utilities|FlowControl|Sequence
        (:then_0
          ;[1 1 1] Adding listeners for Events
          ;; Instrument Changes
          (bind _asynctaskproxy (GameplayEventRouter|StartListeningToEvent OwningPlayerController (Variables|Default|GetEvent%20Router%20Ref) "/Script/PilgrimCoreRuntime.PilgrimEvent_PlayerState_InstrumentChanged" PlayerState)
            (:then
              (Sparks|ViewModel|TrackEventHandle _asynctaskproxy)
              ;; Detecting peripheral connection
              (bind _asynctaskproxy_2 (GameplayEventRouter|StartListeningToEvent OwningPlayerController (Variables|Default|GetEvent%20Router%20Ref) "/Script/PilgrimCoreRuntime.PilgrimEvent_PlayerState_HasConnectedPeripheralChanged" PlayerState)
                (:then
                  (Sparks|ViewModel|TrackEventHandle _asynctaskproxy_2)
                  ;; Changes made by the player in settings, setlist, etc
                  (bind _asynctaskproxy_3 (GameplayEventRouter|StartListeningToEvent OwningPlayerController (Variables|Default|GetEvent%20Router%20Ref) "/Script/PilgrimCoreRuntime.PilgrimEvent_BandState_CustomSetlistChanged")
                    (:then
                      (Sparks|ViewModel|TrackEventHandle _asynctaskproxy_3)
                      (bind _asynctaskproxy_4 (GameplayEventRouter|StartListeningToEvent OwningPlayerController (Variables|Default|GetEvent%20Router%20Ref) "/Script/PilgrimCoreRuntime.PilgrimEvent_PartPickerOpen")
                        (:then
                          (Sparks|ViewModel|TrackEventHandle _asynctaskproxy_4)
                          (bind _asynctaskproxy_5 (GameplayEventRouter|StartListeningToEvent OwningPlayerController (Variables|Default|GetEvent%20Router%20Ref) "/Script/PilgrimSettingsRuntime.PilgrimEvent_PilgrimSettings_AlwaysShowProInstrumentTrackSelection")
                            (:then
                              (Sparks|ViewModel|TrackEventHandle _asynctaskproxy_5)
                              ;; Setting Top Notification status
                              (bind (_asynctaskproxy_6 _payload) (GameplayEventRouter|StartListeningToEvent OwningPlayerController (Variables|Default|GetEvent%20Router%20Ref) "/Script/PilgrimCoreRuntime.PilgrimEvent_PartPickerProVocalsSelected")
                                (:then
                                  (Sparks|ViewModel|TrackEventHandle _asynctaskproxy_6))
                                (:OnEventReceived
                                  (Variables|Default|SetbShowTopNotification (Utilities|Struct|BreakPilgrimEventPartPickerProVocalsSelected _payload)))))
                            (:OnEventReceived
                              (CallFunction|SettingListData))))
                        (:OnEventReceived
                          (CallFunction|SettingListData))))
                    (:OnEventReceived
                      (Default|SettingSongMetaData))))
                (:OnEventReceived
                  (CallFunction|SettingListData))))
            (:OnEventReceived
              (Default|SettingSongMetaData)
              (CallFunction|SettingListData))))
        (:then_1
          ;[0.04795 0.01854 0.15 0.5] This part is to check the BattleStage Competitive case (one Instrument session)
          (bind (_gamestate _playlist) (Class|FortAsyncActionCurrentPlaylistReady|PlaylistDataReadyAsync OwningPlayerController)
            (:OnReady
              (bind _asfort_playlist_athena (Utilities|Casting|CastToFortPlaylistAthena _playlist))
              (bind _output_get (Variables|Default|SetbIsBattleStageCompetitive (GameplayTags|HasTag (Variables|/Script/FortniteGame.FortPlaylist|GetGameplayTagContainer _asfort_playlist_athena) "(TagName=\"Sparks.Playlist.Pilgrim.BattleStage.ScoreChase\")")))
              (if _output_get
                (Variables|Default|SetCurrent%20Tag%20Container (Variables|/Script/FortniteGame.FortPlaylist|GetGameplayTagContainer _asfort_playlist_athena))
                (Default|SettingSongMetaData)
                (CallFunction|SettingListData)))))
        (:then_2
          (Default|SettingSongMetaData)
          (CallFunction|SettingListData))))))

(event UniversalParts|ViewModel|EventBP_OnInputMethodChanged
  (params (NewInputType :type ECommonInputType%20Enum :default "MouseAndKeyboard"))
  (if (CallFunction|IsUsingMobileTraits)
    (CallFunction|SettingListData)))

(fn AddTrackTypeEntry
  (params (InstrumentType :type EPilgrimTrackType%20Enum :default "TrackGuitar") (InstrumentIndex :type Integer) (ListOfInstrumentsData :type Array%20of%20BP%20Pilgrim%20Quickplay%20Part%20Picker%20Instrument%20Entry%20Data%20Object%20References) (DefaultPartSelected :type EPilgrimTrackType%20Enum :default "TrackGuitar"))
  (bind _returnvalue (Game|ConstructObjectfromClass "/PilgrimQuickplay/Widgets/UniversalPartPicker/BP_PilgrimQuickplay_PartPicker_InstrumentEntryData.BP_PilgrimQuickplay_PartPicker_InstrumentEntryData_C" 0 InstrumentType "(INVALID)" (Variables|Default|GetbIsIntensityAvailable) (CallFunction|DetermineIntensitybyTrack (Variables|Default|GetCurrentSongMetaData) InstrumentType) (UI|Instrument|GetInstrumentIconIndex InstrumentType (Variables|Default|GetCurrentSongMetaData) (Class|PilgrimCoreBPFL|PilgrimTrackTypetoSparksInstrument InstrumentType) true self) (UI|Instrument|GetInstrumentNamefromPart InstrumentType self) InstrumentIndex (Variables|Default|GetCurrentSongMetaData)))
  (bind _returnvalue_2 (Utilities|Array|Add (Variables|GetListOfInstrumentsData) _returnvalue))
  ; We check if this entry is the current selected part
  (if (Utilities|Enum|Equal(Enum) InstrumentType (Variables|GetDefaultPartSelected))
    (Variables|Default|SetIndexOnListCurrentInstrument _returnvalue_2)))

(fn CreateListInstruments
  (bind _self self)
  (bind _returnvalue (Pilgrim|Peripherals|IsProDrumsCymbalsEnabled _self))
  (bind _currentpart (CallFunction|GetCurrentInstrumentSelection))
  (LocalVariables|SetCurrentPartSelected _currentpart)
  (Utilities|IsValid (Sparks|ViewModel|BPGetOwningPlayerController)
    (:Is%20Valid
      (if (not (Variables|Default|GetbIsBattleStageCompetitive))
        (if (CallFunction|DetermineProPartsActive)
          ; Determine which parts to add
          (bind _listavailableparts (CallFunction|DetermineAllAvailableParts))
          (LocalVariables|SetListOfPartsToCreate _listavailableparts)
          (else
            (bind _listavailableparts_2 (CallFunction|DetermineDefaultAvailableParts :!label _label))
            (LocalVariables|SetListOfPartsToCreate _listavailableparts_2)))
        (else
          ; (BattleStage One Instrument mode) Setting the list for this case here
          (bind _listpartsbattlestagecompetitive (CallFunction|DeterminePartsBattleStageCompetitive (Variables|Default|GetCurrent%20Tag%20Container)))
          (LocalVariables|SetListOfPartsToCreate _listpartsbattlestagecompetitive)))
      ; Add track type entries to UPP
      (for (_array_element _array_index) (LocalVariables|GetListOfPartsToCreate)
        (switch Utilities|FlowControl|Switch|SwitchonEPilgrimTrackType _array_element
          (:TrackGuitar
            _)
          (:TrackBass
            _)
          (:TrackVocals
            _)
          (:TrackDrum
            _)
          (:TrackPlasticGuitar
            _)
          (:TrackPlasticBass
            _)
          (:TrackPlasticDrum
            ;[0.0365 0.06386 0.15 0.5] ProDrums is a special case because it can be only Plastic Drums or with Cymbals. But both enums refers to the same instrument. We need to choose only one of them
            (if _returnvalue
              _
              (else
                _)))
          (:TrackProVocals
            _)
          (:TrackProCymbals
            (if _returnvalue
              _)))
        (CallFunction|AddTrackTypeEntry _array_element _array_index (LocalVariables|GetListOfInstrumentsData) (LocalVariables|GetCurrentPartSelected)))
      (return (LocalVariables|GetListOfInstrumentsData)))
    (:Is%20Not%20Valid
      _))
  (goto _label))

(fn DetermineAllAvailableParts
  (bind _listavailableparts (CallFunction|DetermineDefaultAvailableParts))
  (LocalVariables|SetListOfPartsAvailable _listavailableparts)
  ;[1 0.9412 0 0.5] Hiding ProParts for Touch input
  (if (not (CallFunction|IsUsingMobileTraits))
    (for _array_element (Variables|Config|GetListProParts)
      (switch Utilities|FlowControl|Switch|SwitchonEPilgrimTrackType _array_element
        (:TrackPlasticGuitar
          _)
        (:TrackPlasticBass
          _)
        (:TrackPlasticDrum
          (if (CallFunction|ShouldShowProDrums) :!label _label
            _))
        (:TrackProCymbals
          (goto _label)))
      (Utilities|Array|Add (LocalVariables|GetListOfPartsAvailable) _array_element)))
  (return (LocalVariables|GetListOfPartsAvailable)))

(fn DetermineDefaultAvailableParts
  (LocalVariables|SetListOfPartsAvailable (Variables|Config|GetListDefaultParts))
  (if (not (CallFunction|ShouldShowProVocals))
    (Utilities|Array|RemoveItem (LocalVariables|GetListOfPartsAvailable) "TrackProVocals"))
  (return (LocalVariables|GetListOfPartsAvailable)))

(fn DetermineIntensityByTrack
  (params (SongMetaData :type Pilgrim%20Song%20Metadata%20Object%20Reference) (TrackType :type EPilgrimTrackType%20Enum :default "TrackGuitar"))
  (Utilities|IsValid SongMetaData
    (:Is%20Valid
      (return (Class|PilgrimSongMetadata|GetTrackIntensity SongMetaData TrackType)))))

(fn DeterminePartsBattleStageCompetitive
  (params (GamePlayTagContainer :type Gameplay%20Tag%20Container%20Structure))
  (if (GameplayTags|HasTag GamePlayTagContainer "(TagName=\"Sparks.Playlist.Pilgrim.BattleStage.ScoreChase.Bass\")" true)
    (LocalVariables|SetPartsToCreate (Utilities|Array|MakeArray "TrackBass"))
    (if (CallFunction|DetermineProPartsActive)
      (Utilities|Array|Add (LocalVariables|GetPartsToCreate) "TrackPlasticBass"))
    (return (LocalVariables|GetPartsToCreate))
    (elif (GameplayTags|HasTag GamePlayTagContainer "(TagName=\"Sparks.Playlist.Pilgrim.BattleStage.ScoreChase.Drums\")" true)
      (LocalVariables|SetPartsToCreate (Utilities|Array|MakeArray "TrackDrum"))
      (if (and (CallFunction|ShouldShowProDrums) (CallFunction|DetermineProPartsActive))
        (Utilities|Array|Add (LocalVariables|GetPartsToCreate) "TrackPlasticDrum"))
      (return (LocalVariables|GetPartsToCreate))
      (elif (GameplayTags|HasTag GamePlayTagContainer "(TagName=\"Sparks.Playlist.Pilgrim.BattleStage.ScoreChase.Lead\")" true)
        (LocalVariables|SetPartsToCreate (Utilities|Array|MakeArray "TrackGuitar"))
        (if (CallFunction|DetermineProPartsActive)
          (Utilities|Array|Add (LocalVariables|GetPartsToCreate) "TrackPlasticGuitar"))
        (return (LocalVariables|GetPartsToCreate))
        (elif (GameplayTags|HasTag GamePlayTagContainer "(TagName=\"Sparks.Playlist.Pilgrim.BattleStage.ScoreChase.Vocals\")" true)
          (LocalVariables|SetPartsToCreate (Utilities|Array|MakeArray "TrackVocals"))
          (return (LocalVariables|GetPartsToCreate)))))))

(fn DetermineProPartsActive
  (bind _returnvalue (Sparks|ViewModel|BPGetOwningPlayerController))
  (bind _returnvalue_2 (Actor|GetComponentbyClass _returnvalue "/PilgrimCore/Player/Components/BP_PilgrimPlayerControllerComponent.BP_PilgrimPlayerControllerComponent_C"))
  (Utilities|IsValid _returnvalue
    (:Is%20Valid
      (Utilities|IsValid _returnvalue_2
        (:Is%20Valid
          (return (Class|PilgrimPlayerControllerComponent|CanProInstrumentsbeSelected _returnvalue_2)))))))

(fn GetInstrumentText
  (return (UI|Instrument|GetInstrumentNamefromPart (Variables|Default|GetInstrumentChanged) self)))

(fn GetCurrentInstrumentSelection
  (bind _returnvalue (Sparks|ViewModel|BPGetOwningPlayerController))
  (bind _playerstate (Variables|/Script/Engine.Controller|GetPlayerState _returnvalue))
  (bind _returnvalue_2 (Actor|GetComponentbyClass _playerstate "/Script/PilgrimCoreRuntime.PilgrimCorePlayerStateComponent"))
  (Utilities|IsValid _returnvalue_2
    (:Is%20Valid
      (return (Class|PilgrimCorePlayerStateComponent|GetInstrumentChoice _returnvalue_2)))
    (:Is%20Not%20Valid
      (return "TrackVocals"))))

(fn GetInstrumentIconIndex
  (return (UI|Instrument|GetInstrumentIconIndex (Variables|Default|GetInstrumentChanged) (Variables|Default|GetCurrentSongMetaData) (Class|PilgrimCoreBPFL|PilgrimTrackTypetoSparksInstrument (Variables|Default|GetInstrumentChanged)) true self)))

(fn GetIntensityIndex
  (bind _currentsongmetadata (Variables|Default|GetCurrentSongMetaData))
  (Utilities|IsValid _currentsongmetadata
    (:Is%20Valid
      (return (Class|PilgrimSongMetadata|GetTrackIntensity _currentsongmetadata (Variables|Default|GetInstrumentChanged))))
    (:Is%20Not%20Valid
      (return 0))))

(fn IsUsingMobileTraits
  (return (and (Inputs|IsUsingInputType "Touch" (Sparks|ViewModel|BPGetOwningPlayerController)) (DisplayHelper|IsUsingCompactUI))))

(fn SetListPartsInfo
  (bind _listallparts (Variables|Config|GetListAllParts))
  (Utilities|Array|Clear _listallparts)
  (Utilities|Array|AppendArray _listallparts (Variables|Config|GetListDefaultParts))
  (Utilities|Array|AppendArray _listallparts (Variables|Config|GetListProParts))
  (Variables|Config|SetListAllParts _listallparts))

(fn SettingListData
  (bind _listavailableinstruments (CallFunction|CreateListInstruments))
  (Variables|Default|SetListOfIntrumentsInfo _listavailableinstruments))

(fn SettingSongMetaData
  (bind _returnvalue (Sparks|ViewModel|BPGetOwningPlayerController))
  (bind _returnvalue_2 (Class|PilgrimQuickplayBPFL|GetUpcomingSong _returnvalue))
  (Utilities|IsValid _returnvalue
    (:Is%20Valid
      (if (Utilities|IsValid _returnvalue_2 :!class /Script/Engine.KismetSystemLibrary)
        (Variables|Default|SetbIsIntensityAvailable true)
        (Variables|Default|SetCurrentSongMetaData _returnvalue_2)
        (else
          _)))
    (:Is%20Not%20Valid
      _))
  (Variables|Default|SetbIsIntensityAvailable false))

(fn ShouldShowProDrums
  (return (Class|PilgrimPlayerControllerComponent|IsChoppedCheeseEnabled (Actor|GetComponentbyClass (Sparks|ViewModel|BPGetOwningPlayerController) "/PilgrimCore/Player/Components/BP_PilgrimPlayerControllerComponent.BP_PilgrimPlayerControllerComponent_C"))))

(fn ShouldShowProVocals
  (return (Class|PilgrimCorePlayerStateComponent|IsChickenParmEnabled (Actor|GetComponentbyClass (Variables|/Script/Engine.Controller|GetPlayerState (Sparks|ViewModel|BPGetOwningPlayerController)) "/Script/PilgrimCoreRuntime.PilgrimCorePlayerStateComponent"))))
